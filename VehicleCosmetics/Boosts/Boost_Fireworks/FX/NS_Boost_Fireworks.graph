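# Niagara system for the Fireworks vehicle boost cosmetic, emitting sparks, sparkles and bursts from filtered skeletal mesh sockets.
# GPU and CPU fill emitters merge three particle behaviors each through spawn groups, alongside a looping GPU start burst, inner/outer sparkle layers and a large-burst parent emitter.
# User.ColorVariantID drives painted color overrides, with random gradient colors as the default in several emitters.
# User.bIsFrontendPreview retunes rates, speeds, lifetimes and radii for frontend preview.
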NiagaraSystem /VehicleCosmetics/Boosts/Boost_Fireworks/FX/NS_Boost_Fireworks
dependencies {
 usedDataInterfaces [
  &NiagaraDataInterfaceColorCurve
  &NiagaraDataInterfaceCurve
  &NiagaraDataInterfaceArrayColor
  &NiagaraDataInterfaceArrayFloat
  &NiagaraDataInterfaceArrayFloat2
  &NiagaraDataInterfaceVector2DCurve
  &NiagaraDataInterfaceParticleRead
  &NiagaraDataInterfaceSkeletalMesh
 ]
 usedDynamicInputs [
  &MakeCustomLinearColorFromBool
  &SelectLinearColorFromArray
  &MultiplyLinearColor
  &MakeVector4_FromLinearColor
  /VehicleCosmetics/FX/Niagara/Modules/MergeEmitters/VC_Dynamic_ColorBySpawnGroup
  &RandomRangeLinearColor
 ]
 usedModules [
  &self:SystemSpawnScript.SystemScriptSource.SystemScriptGraph.NiagaraNodeAssignment_0.SetVariables_#1
  &SystemState
 ]
}
emitters {
 NE_Firework_Fill {
  simTarget=GPUComputeSim
  emitterUpdateScript [
   &EmitterState()
   &VC_FilteredSocketCount()
   &SpawnRate(SpawnRate=150, Use Spawn Probability=false)
   &SpawnRate(
    SpawnGroup=1
    SpawnRate=&MakeCustomFloatFromBool(Bool=$User.bIsFrontendPreview, True Float=25, False Float=100)
    Use Spawn Probability=false
   ) {module=SpawnRate001}
   &SpawnRate(
    SpawnGroup=2
    SpawnRate=&MakeCustomFloatFromBool(Bool=$User.bIsFrontendPreview, True Float=700, False Float=1500)
    Use Spawn Probability=false
   ) {module=SpawnRate002}
  ]
  particleSpawnScript [
   &self:ColorVariantID_Material_Select(VariantID:NiagaraInt32=$User.ColorVariantID) {module=ColorVariantID_Material_Select}
   &VC_PaintedVariantsOverride(
    Black=<rgba 0 0 0 10>
    BurntSienna=&Multiply_LinearColorByFloat(Linear Color=<rgba 0.25 0.1022 0.0625 1>, Float=50)
    Cobalt=&Multiply_LinearColorByFloat(Linear Color=<rgba 0 0.0185 1 1>, Float=50)
    Crimson=&Multiply_LinearColorByFloat(Linear Color=<rgba 0.4635 0 0.0126 1>, Float=50)
    DefaultColor=&Multiply_LinearColorByFloat(
     Linear Color=&ColorFromCurve(
      ColorCurve=DataInterface() {propertyValues {
        AlphaCurve {keys [{interpMode=RCIM_Cubic, tangentMode=RCTM_User, time=0, value=1}]}
        BlueCurve {keys [{leaveTangentWeight=0.1167, time=0.1, value=0.005}, <at 0.45 1>, <at 0.55 1>, <at 0.9 1>]}
        GreenCurve {keys [{leaveTangentWeight=0.1167, time=0.1, value=0}, <at 0.45 0>, <at 0.55 0.0268>, <at 0.9 0.1>]}
        RedCurve {keys [<at 0.1 1>,<at 0.45 0.3333>,<at 0.55 0.2439>,<at 0.9 0>]}
       }}
      CurveIndex=&RandomRangeFloat(Minimum=0, Maximum=1)
     )
     Float=/VehicleCosmetics/FX/Niagara/Modules/MergeEmitters/VC_FloatBySpawnGroup(
      SpawnGroup or ExecIndex:NiagaraInt32=0
      CustomInt:NiagaraInt32=0
      MaximumSpawnGroups:NiagaraInt32=3
      Float SG 0:NiagaraFloat=50
      Float SG 1:NiagaraFloat=50
      Float SG 2:NiagaraFloat=50
      Float SG 3:NiagaraFloat=0
      Float SG 4:NiagaraFloat=0
      Float SG 5:NiagaraFloat=0
     )
    )
    ForestGreen=&Multiply_LinearColorByFloat(Linear Color=<rgba 0 0.5 0.0253 1>, Float=50)
    Grey=<rgba 0.05 0.05 0.05 10>
    Lime=&Multiply_LinearColorByFloat(Linear Color=<rgba 0.2943 0.5885 0 1>, Float=50)
    Orange=&Multiply_LinearColorByFloat(Linear Color=<rgba 0.6724 0.2346 0 1>, Float=50)
    Pink=&Multiply_LinearColorByFloat(Linear Color=<rgba 0.5052 0.0605 0.3435 1>, Float=50)
    Purple=&Multiply_LinearColorByFloat(Linear Color=<rgba 0.2789 0 0.6795 1>, Float=50)
    Saffron=&Multiply_LinearColorByFloat(Linear Color=<rgba 0.9 0.69 0 1>, Float=50)
    SkyBlue=&Multiply_LinearColorByFloat(Linear Color=<rgba 0 0.357 0.5365 1>, Float=50)
    TitaniumWhite=&Multiply_LinearColorByFloat(Linear Color=<rgba 0.75 0.75 0.75 1>, Float=50)
   )
   &InitializeParticle(
    Color Mode=::NewEnumerator0
    Lifetime Max=0.4
    Lifetime Min=0.2
    Lifetime Mode=::NewEnumerator1
    Sprite Size Mode=::NewEnumerator3
    Uniform Sprite Size=&VC_FloatBySpawnGroup(
     SpawnGroup or ExecIndex:NiagaraInt32=0
     CustomInt:NiagaraInt32=0
     MaximumSpawnGroups:NiagaraInt32=3
     Float SG 0:NiagaraFloat=&RandomRangeFloat(Minimum=4, Maximum=7)
     Float SG 1:NiagaraFloat=&RandomRangeFloat(Minimum=8, Maximum=10)
     Float SG 2:NiagaraFloat=&RandomRangeFloat(Minimum=4, Maximum=5)
     Float SG 3:NiagaraFloat=&RandomRangeFloat(Minimum=3, Maximum=10)
     Float SG 4:NiagaraFloat=&RandomRangeFloat(Minimum=4, Maximum=5)
     Float SG 5:NiagaraFloat=0
    )
    UsePositionOffset=false
    Write Material Random=true
   )
   &PartitionParticles(Partition Mode:ENiagara_PartitionMode_v2=::NewEnumerator5)
   &SkeletalMeshLocation(
    Bone / Socket Index=$Particles.PartitionParticles.Partition
    Bone / Socket Sampling Mode=::NewEnumerator5
    MaskSpawnGroup=false
    Mesh Sampling Type=::NewEnumerator6
    Skeletal Mesh=$System.SkeletalMesh
   )
   &ShapeLocation(
    MaskSpawnGroup=false
    Override Seed=false
    Sphere Radius=&VC_FloatBySpawnGroup(
     SpawnGroup or ExecIndex:NiagaraInt32=0
     CustomInt:NiagaraInt32=0
     MaximumSpawnGroups:NiagaraInt32=3
     Float SG 0:NiagaraFloat=&MakeCustomFloatFromBool(Bool=$User.bIsFrontendPreview, True Float=10, False Float=45)
     Float SG 1:NiagaraFloat=&MakeCustomFloatFromBool(Bool=$User.bIsFrontendPreview, True Float=15, False Float=45)
     Float SG 2:NiagaraFloat=&MakeCustomFloatFromBool(Bool=$User.bIsFrontendPreview, True Float=20, False Float=45)
     Float SG 3:NiagaraFloat=10
     Float SG 4:NiagaraFloat=5
     Float SG 5:NiagaraFloat=0
    )
    UseOwnerScale=true
   )
   &AddVelocity(
    Cone Angle=&SelectFloatFromArray(
     Direct Array Index=&SpawnGroup()
     Float Selection Array=DataInterface() {propertyValues {FloatData [150,50,180]}}
    )
    Cone Axis=<-1 0 0>
    Distribution Along Cone Axis=0.8
    Module.UseConeFalloff=true
    Module.UseVelDistribution=true
    Override Seed=false
    Speed Falloff From Cone Axis=0.75
    Velocity Mode=::NewEnumerator2
    Velocity Speed=&VC_FloatBySpawnGroup(
     SpawnGroup or ExecIndex:NiagaraInt32=0
     CustomInt:NiagaraInt32=0
     MaximumSpawnGroups:NiagaraInt32=3
     Float SG 0:NiagaraFloat=&MakeCustomFloatFromBool(
      Bool=$User.bIsFrontendPreview
      True Float=&RandomRangeFloat(Minimum=2000, Maximum=5000)
      False Float=&RandomRangeFloat(Minimum=350, Maximum=1000)
     )
     Float SG 1:NiagaraFloat=&MakeCustomFloatFromBool(
      Bool=$User.bIsFrontendPreview
      True Float=&RandomRangeFloat(Minimum=2000, Maximum=2500)
      False Float=&RandomRangeFloat(Minimum=200, Maximum=400)
     )
     Float SG 2:NiagaraFloat=&MakeCustomFloatFromBool(
      Bool=$User.bIsFrontendPreview
      True Float=&RandomRangeFloat(Minimum=1200, Maximum=5000)
      False Float=&RandomRangeFloat(Minimum=0, Maximum=0)
     )
     Float SG 3:NiagaraFloat=10
     Float SG 4:NiagaraFloat=5
     Float SG 5:NiagaraFloat=0
    )
   )
  ]
  particleUpdateScript [
   &ParticleState()
   &Drag(
    Drag=&SelectFloatFromArray(Direct Array Index=&SpawnGroup(), Float Selection Array=DataInterface() {propertyValues {FloatData [5,1,15]}})
    Use Linear Drag=true
    Use Rotational Drag=true
   )
   &CurlNoiseForce(
    Bias Noise=false
    Noise Frequency=&VC_FloatBySpawnGroup(
     SpawnGroup or ExecIndex:NiagaraInt32=0
     CustomInt:NiagaraInt32=0
     MaximumSpawnGroups:NiagaraInt32=3
     Float SG 0:NiagaraFloat=50
     Float SG 1:NiagaraFloat=50
     Float SG 2:NiagaraFloat=50
     Float SG 3:NiagaraFloat=0
     Float SG 4:NiagaraFloat=50
     Float SG 5:NiagaraFloat=0
    )
    Noise Strength=&VC_FloatBySpawnGroup(
     SpawnGroup or ExecIndex:NiagaraInt32=0
     CustomInt:NiagaraInt32=0
     MaximumSpawnGroups:NiagaraInt32=3
     Float SG 0:NiagaraFloat=100
     Float SG 1:NiagaraFloat=&MakeCustomFloatFromBool(Bool=$User.bIsFrontendPreview, True Float=1000, False Float=0)
     Float SG 2:NiagaraFloat=2000
     Float SG 3:NiagaraFloat=0
     Float SG 4:NiagaraFloat=0
     Float SG 5:NiagaraFloat=0
    )
    Pan Noise=false
   )
   &GravityForce(Gravity=&Multiply_VectorByFloat(
     Vector=<1 1 -980>
     Float=&SelectFloatFromArray(Direct Array Index=&SpawnGroup(), Float Selection Array=DataInterface() {propertyValues {FloatData [0,0.25,0]}})
    ))
   &SolveForcesAndVelocity(Clamp Velocity=false, Limit Acceleration=false)
   &ScaleSpriteSizeBySpeed(
    Max Scale Factor=&SelectVector2DFromArray(
     Direct Array Index=&SpawnGroup()
     Vector 2D Selection Array=DataInterface() {propertyValues {FloatData [<0.5 3>, <0.5 7>, <0.3 2>]}}
    )
    Min Scale Factor=&SelectVector2DFromArray(
     Direct Array Index=&SpawnGroup()
     Vector 2D Selection Array=DataInterface() {propertyValues {FloatData [<1 1>, <1 2>, <0.75 0.75>]}}
    )
    Velocity Threshold:NiagaraFloat=&SelectFloatFromArray(
     Direct Array Index=&SpawnGroup()
     Float Selection Array=DataInterface() {propertyValues {FloatData [100,450,900]}}
    )
   )
   &ScaleSpriteSize(
    Scale Sprite Size Mode=::NewEnumerator0
    Uniform Scale Factor=&VC_FloatBySpawnGroup(
     SpawnGroup or ExecIndex:NiagaraInt32=0
     CustomInt:NiagaraInt32=0
     MaximumSpawnGroups:NiagaraInt32=3
     Float SG 0:NiagaraFloat=1
     Float SG 1:NiagaraFloat=&FloatFromCurve(FloatCurve=DataInterface() {propertyValues {Curve {
         defaultValue=1
         keys [
          {arriveTangent=9.0725, interpMode=RCIM_Cubic, leaveTangent=9.0725, tangentMode=RCTM_User, time=0, value=0}
          {arriveTangentWeight=0.1274, interpMode=RCIM_Cubic, leaveTangentWeight=0.7778, tangentMode=RCTM_User, time=0.1408, value=1}
         ]
        }}})
     Float SG 2:NiagaraFloat=&FloatFromCurve(FloatCurve=DataInterface() {propertyValues {Curve {
         defaultValue=1
         keys [
          {tangentMode=RCTM_User, time=0, value=1}
          {arriveTangent=-0.0405, interpMode=RCIM_Cubic, leaveTangent=-0.0405, tangentMode=RCTM_User, time=0.7009, value=1}
          {arriveTangent=-8.0089, interpMode=RCIM_Cubic, leaveTangent=-8.0089, tangentMode=RCTM_User, time=1, value=0}
         ]
        }}})
     Float SG 3:NiagaraFloat=&FloatFromCurve(FloatCurve=DataInterface() {propertyValues {Curve {
         defaultValue=1
         keys [
          {arriveTangent=9.0725, interpMode=RCIM_Cubic, leaveTangent=9.0725, tangentMode=RCTM_User, time=0, value=0}
          {arriveTangentWeight=0.1274, interpMode=RCIM_Cubic, leaveTangentWeight=0.7778, tangentMode=RCTM_User, time=0.1408, value=0.9847}
         ]
        }}})
     Float SG 4:NiagaraFloat=&FloatFromCurve(FloatCurve=DataInterface() {propertyValues {Curve {
         defaultValue=1
         keys [
          {tangentMode=RCTM_User, time=0, value=1}
          {arriveTangent=-0.0405, interpMode=RCIM_Cubic, leaveTangent=-0.0405, tangentMode=RCTM_User, time=0.7009, value=1}
          {arriveTangent=-8.0089, interpMode=RCIM_Cubic, leaveTangent=-8.0089, tangentMode=RCTM_User, time=1, value=0}
         ]
        }}})
     Float SG 5:NiagaraFloat=0
    )
   )
   &ScaleColor(
    Scale Alpha=&VC_FloatBySpawnGroup(
     SpawnGroup or ExecIndex:NiagaraInt32=0
     CustomInt:NiagaraInt32=0
     MaximumSpawnGroups:NiagaraInt32=3
     Float SG 0:NiagaraFloat=&FloatFromCurve(FloatCurve=DataInterface() {propertyValues {Curve {
         defaultValue=1
         keys [{interpMode=RCIM_Cubic, tangentMode=RCTM_User, time=0, value=1}, {interpMode=RCIM_Cubic, tangentMode=RCTM_User, time=1, value=0}]
        }}})
     Float SG 1:NiagaraFloat=&FloatFromCurve(FloatCurve=DataInterface() {propertyValues {Curve {
         defaultValue=1
         keys [
          {interpMode=RCIM_Cubic, tangentMode=RCTM_User, time=0, value=1}
          {arriveTangent=-2.8897, interpMode=RCIM_Cubic, leaveTangent=-2.8897, tangentMode=RCTM_User, time=1, value=0}
         ]
        }}})
     Float SG 2:NiagaraFloat=1
     Float SG 3:NiagaraFloat=&RandomRangeFloat(Minimum=3, Maximum=10)
     Float SG 4:NiagaraFloat=&RandomRangeFloat(Minimum=4, Maximum=5)
     Float SG 5:NiagaraFloat=0
    )
    ScaleA=true
    ScaleRGB=true
   )
  ]
  rendererClasses [&NiagaraSpriteRendererProperties]
  renderers [
   {rendererClass=&NiagaraSpriteRendererProperties}
   {rendererClass=&NiagaraSpriteRendererProperties, rendererIndex=1}
  ]
 }
 NE_Firework_Fill001 {
  emitterUpdateScript [
   &EmitterState()
   &VC_FilteredSocketCount()
   &SpawnRate(SpawnRate=25, Use Spawn Probability=false)
   &SpawnRate(
    SpawnGroup=1
    SpawnRate=&MakeCustomFloatFromBool(Bool=$User.bIsFrontendPreview, True Float=20, False Float=25)
    Use Spawn Probability=false
   ) {module=SpawnRate001}
   &SpawnRate(
    SpawnGroup=2
    SpawnRate=&MakeCustomFloatFromBool(Bool=$User.bIsFrontendPreview, True Float=70, False Float=90)
    Use Spawn Probability=false
   ) {module=SpawnRate002}
  ]
  particleSpawnScript [
   &self:ColorVariantID_Material_Select(VariantID:NiagaraInt32=$User.ColorVariantID) {module=ColorVariantID_Material_Select}
   &VC_PaintedVariantsOverride(
    Black=<rgba 0 0 0 10>
    BurntSienna=&Multiply_LinearColorByFloat(Linear Color=<rgba 0.25 0.1022 0.0625 1>, Float=50)
    Cobalt=&Multiply_LinearColorByFloat(Linear Color=<rgba 0 0.0185 1 1>, Float=50)
    Crimson=&Multiply_LinearColorByFloat(Linear Color=<rgba 0.4635 0 0.0126 1>, Float=50)
    DefaultColor=&Multiply_LinearColorByFloat(
     Linear Color=&ColorFromCurve(
      ColorCurve=DataInterface() {propertyValues {
        AlphaCurve {keys [{interpMode=RCIM_Cubic, tangentMode=RCTM_User, time=0, value=1}]}
        BlueCurve {keys [{leaveTangentWeight=0.1167, time=0.1, value=0.005}, <at 0.45 1>, <at 0.55 1>, <at 0.9 1>]}
        GreenCurve {keys [{leaveTangentWeight=0.1167, time=0.1, value=0}, <at 0.45 0>, <at 0.55 0.0268>, <at 0.9 0.1>]}
        RedCurve {keys [<at 0.1 1>,<at 0.45 0.3333>,<at 0.55 0.2439>,<at 0.9 0>]}
       }}
      CurveIndex=&RandomRangeFloat(Minimum=0, Maximum=1)
     )
     Float=&VC_FloatBySpawnGroup(
      SpawnGroup or ExecIndex:NiagaraInt32=0
      CustomInt:NiagaraInt32=0
      MaximumSpawnGroups:NiagaraInt32=3
      Float SG 0:NiagaraFloat=50
      Float SG 1:NiagaraFloat=50
      Float SG 2:NiagaraFloat=50
      Float SG 3:NiagaraFloat=0
      Float SG 4:NiagaraFloat=0
      Float SG 5:NiagaraFloat=0
     )
    )
    ForestGreen=&Multiply_LinearColorByFloat(Linear Color=<rgba 0 0.5 0.0253 1>, Float=50)
    Grey=<rgba 0.05 0.05 0.05 10>
    Lime=&Multiply_LinearColorByFloat(Linear Color=<rgba 0.2943 0.5885 0 1>, Float=50)
    Orange=&Multiply_LinearColorByFloat(Linear Color=<rgba 0.6724 0.2346 0 1>, Float=50)
    Pink=&Multiply_LinearColorByFloat(Linear Color=<rgba 0.5052 0.0605 0.3435 1>, Float=50)
    Purple=&Multiply_LinearColorByFloat(Linear Color=<rgba 0.2789 0 0.6795 1>, Float=50)
    Saffron=&Multiply_LinearColorByFloat(Linear Color=<rgba 0.9 0.69 0 1>, Float=50)
    SkyBlue=&Multiply_LinearColorByFloat(Linear Color=<rgba 0 0.357 0.5365 1>, Float=50)
    TitaniumWhite=&Multiply_LinearColorByFloat(Linear Color=<rgba 0.75 0.75 0.75 1>, Float=50)
   )
   &InitializeParticle(
    Color Mode=::NewEnumerator0
    Lifetime Max=0.4
    Lifetime Min=0.2
    Lifetime Mode=::NewEnumerator1
    Sprite Size Mode=::NewEnumerator3
    Uniform Sprite Size=&VC_FloatBySpawnGroup(
     SpawnGroup or ExecIndex:NiagaraInt32=0
     CustomInt:NiagaraInt32=0
     MaximumSpawnGroups:NiagaraInt32=3
     Float SG 0:NiagaraFloat=&RandomRangeFloat(Minimum=4, Maximum=7)
     Float SG 1:NiagaraFloat=&RandomRangeFloat(Minimum=8, Maximum=10)
     Float SG 2:NiagaraFloat=&RandomRangeFloat(Minimum=4, Maximum=5)
     Float SG 3:NiagaraFloat=&RandomRangeFloat(Minimum=3, Maximum=10)
     Float SG 4:NiagaraFloat=&RandomRangeFloat(Minimum=4, Maximum=5)
     Float SG 5:NiagaraFloat=0
    )
    UsePositionOffset=false
    Write Material Random=true
   )
   &PartitionParticles(Partition Mode:ENiagara_PartitionMode_v2=::NewEnumerator5)
   &SkeletalMeshLocation(
    Bone / Socket Index=$Particles.PartitionParticles.Partition
    Bone / Socket Sampling Mode=::NewEnumerator5
    MaskSpawnGroup=false
    Mesh Sampling Type=::NewEnumerator6
    Skeletal Mesh=$System.SkeletalMesh
   )
   &ShapeLocation(
    MaskSpawnGroup=false
    Override Seed=false
    Sphere Radius=&VC_FloatBySpawnGroup(
     SpawnGroup or ExecIndex:NiagaraInt32=0
     CustomInt:NiagaraInt32=0
     MaximumSpawnGroups:NiagaraInt32=3
     Float SG 0:NiagaraFloat=&MakeCustomFloatFromBool(Bool=$User.bIsFrontendPreview, True Float=10, False Float=45)
     Float SG 1:NiagaraFloat=&MakeCustomFloatFromBool(Bool=$User.bIsFrontendPreview, True Float=15, False Float=45)
     Float SG 2:NiagaraFloat=&MakeCustomFloatFromBool(Bool=$User.bIsFrontendPreview, True Float=20, False Float=45)
     Float SG 3:NiagaraFloat=10
     Float SG 4:NiagaraFloat=5
     Float SG 5:NiagaraFloat=0
    )
    UseOwnerScale=true
   )
   &AddVelocity(
    Cone Angle=&SelectFloatFromArray(
     Direct Array Index=&SpawnGroup()
     Float Selection Array=DataInterface() {propertyValues {FloatData [150,50,180]}}
    )
    Cone Axis=<-1 0 0>
    Distribution Along Cone Axis=0.8
    Module.UseConeFalloff=true
    Module.UseVelDistribution=true
    Override Seed=false
    Speed Falloff From Cone Axis=0.75
    Velocity Mode=::NewEnumerator2
    Velocity Speed=&VC_FloatBySpawnGroup(
     SpawnGroup or ExecIndex:NiagaraInt32=0
     CustomInt:NiagaraInt32=0
     MaximumSpawnGroups:NiagaraInt32=3
     Float SG 0:NiagaraFloat=&MakeCustomFloatFromBool(
      Bool=$User.bIsFrontendPreview
      True Float=&RandomRangeFloat(Minimum=2000, Maximum=5000)
      False Float=&RandomRangeFloat(Minimum=350, Maximum=1000)
     )
     Float SG 1:NiagaraFloat=&MakeCustomFloatFromBool(
      Bool=$User.bIsFrontendPreview
      True Float=&RandomRangeFloat(Minimum=2000, Maximum=2500)
      False Float=&RandomRangeFloat(Minimum=200, Maximum=400)
     )
     Float SG 2:NiagaraFloat=&MakeCustomFloatFromBool(
      Bool=$User.bIsFrontendPreview
      True Float=&RandomRangeFloat(Minimum=1200, Maximum=5000)
      False Float=&RandomRangeFloat(Minimum=0, Maximum=0)
     )
     Float SG 3:NiagaraFloat=10
     Float SG 4:NiagaraFloat=5
     Float SG 5:NiagaraFloat=0
    )
   )
  ]
  particleUpdateScript [
   &ParticleState()
   &Drag(
    Drag=&SelectFloatFromArray(Direct Array Index=&SpawnGroup(), Float Selection Array=DataInterface() {propertyValues {FloatData [5,1,15]}})
    Use Linear Drag=true
    Use Rotational Drag=true
   )
   &CurlNoiseForce(
    Bias Noise=false
    Noise Frequency=&VC_FloatBySpawnGroup(
     SpawnGroup or ExecIndex:NiagaraInt32=0
     CustomInt:NiagaraInt32=0
     MaximumSpawnGroups:NiagaraInt32=3
     Float SG 0:NiagaraFloat=50
     Float SG 1:NiagaraFloat=50
     Float SG 2:NiagaraFloat=50
     Float SG 3:NiagaraFloat=0
     Float SG 4:NiagaraFloat=50
     Float SG 5:NiagaraFloat=0
    )
    Noise Strength=&VC_FloatBySpawnGroup(
     SpawnGroup or ExecIndex:NiagaraInt32=0
     CustomInt:NiagaraInt32=0
     MaximumSpawnGroups:NiagaraInt32=3
     Float SG 0:NiagaraFloat=100
     Float SG 1:NiagaraFloat=&MakeCustomFloatFromBool(Bool=$User.bIsFrontendPreview, True Float=1000, False Float=0)
     Float SG 2:NiagaraFloat=2000
     Float SG 3:NiagaraFloat=0
     Float SG 4:NiagaraFloat=0
     Float SG 5:NiagaraFloat=0
    )
    Pan Noise=false
   )
   &GravityForce(Gravity=&Multiply_VectorByFloat(
     Vector=<1 1 -980>
     Float=&SelectFloatFromArray(Direct Array Index=&SpawnGroup(), Float Selection Array=DataInterface() {propertyValues {FloatData [0,0.25,0]}})
    ))
   &SolveForcesAndVelocity(Clamp Velocity=false, Limit Acceleration=false)
   &ScaleSpriteSizeBySpeed(
    Max Scale Factor=&SelectVector2DFromArray(
     Direct Array Index=&SpawnGroup()
     Vector 2D Selection Array=DataInterface() {propertyValues {FloatData [<0.5 3>, <0.5 7>, <0.3 2>]}}
    )
    Min Scale Factor=&SelectVector2DFromArray(
     Direct Array Index=&SpawnGroup()
     Vector 2D Selection Array=DataInterface() {propertyValues {FloatData [<1 1>, <1 2>, <0.75 0.75>]}}
    )
    Velocity Threshold:NiagaraFloat=&SelectFloatFromArray(
     Direct Array Index=&SpawnGroup()
     Float Selection Array=DataInterface() {propertyValues {FloatData [100,450,900]}}
    )
   )
   &ScaleSpriteSize(
    Scale Sprite Size Mode=::NewEnumerator0
    Uniform Scale Factor=&VC_FloatBySpawnGroup(
     SpawnGroup or ExecIndex:NiagaraInt32=0
     CustomInt:NiagaraInt32=0
     MaximumSpawnGroups:NiagaraInt32=3
     Float SG 0:NiagaraFloat=1
     Float SG 1:NiagaraFloat=&FloatFromCurve(FloatCurve=DataInterface() {propertyValues {Curve {
         defaultValue=1
         keys [
          {arriveTangent=9.0725, interpMode=RCIM_Cubic, leaveTangent=9.0725, tangentMode=RCTM_User, time=0, value=0}
          {arriveTangentWeight=0.1274, interpMode=RCIM_Cubic, leaveTangentWeight=0.7778, tangentMode=RCTM_User, time=0.1408, value=1}
         ]
        }}})
     Float SG 2:NiagaraFloat=&FloatFromCurve(FloatCurve=DataInterface() {propertyValues {Curve {
         defaultValue=1
         keys [
          {tangentMode=RCTM_User, time=0, value=1}
          {arriveTangent=-0.0405, interpMode=RCIM_Cubic, leaveTangent=-0.0405, tangentMode=RCTM_User, time=0.7009, value=1}
          {arriveTangent=-8.0089, interpMode=RCIM_Cubic, leaveTangent=-8.0089, tangentMode=RCTM_User, time=1, value=0}
         ]
        }}})
     Float SG 3:NiagaraFloat=&FloatFromCurve(FloatCurve=DataInterface() {propertyValues {Curve {
         defaultValue=1
         keys [
          {arriveTangent=9.0725, interpMode=RCIM_Cubic, leaveTangent=9.0725, tangentMode=RCTM_User, time=0, value=0}
          {arriveTangentWeight=0.1274, interpMode=RCIM_Cubic, leaveTangentWeight=0.7778, tangentMode=RCTM_User, time=0.1408, value=0.9847}
         ]
        }}})
     Float SG 4:NiagaraFloat=&FloatFromCurve(FloatCurve=DataInterface() {propertyValues {Curve {
         defaultValue=1
         keys [
          {tangentMode=RCTM_User, time=0, value=1}
          {arriveTangent=-0.0405, interpMode=RCIM_Cubic, leaveTangent=-0.0405, tangentMode=RCTM_User, time=0.7009, value=1}
          {arriveTangent=-8.0089, interpMode=RCIM_Cubic, leaveTangent=-8.0089, tangentMode=RCTM_User, time=1, value=0}
         ]
        }}})
     Float SG 5:NiagaraFloat=0
    )
   )
   &ScaleColor(
    Scale Alpha=&VC_FloatBySpawnGroup(
     SpawnGroup or ExecIndex:NiagaraInt32=0
     CustomInt:NiagaraInt32=0
     MaximumSpawnGroups:NiagaraInt32=3
     Float SG 0:NiagaraFloat=&FloatFromCurve(FloatCurve=DataInterface() {propertyValues {Curve {
         defaultValue=1
         keys [{interpMode=RCIM_Cubic, tangentMode=RCTM_User, time=0, value=1}, {interpMode=RCIM_Cubic, tangentMode=RCTM_User, time=1, value=0}]
        }}})
     Float SG 1:NiagaraFloat=&FloatFromCurve(FloatCurve=DataInterface() {propertyValues {Curve {
         defaultValue=1
         keys [
          {interpMode=RCIM_Cubic, tangentMode=RCTM_User, time=0, value=1}
          {arriveTangent=-2.8897, interpMode=RCIM_Cubic, leaveTangent=-2.8897, tangentMode=RCTM_User, time=1, value=0}
         ]
        }}})
     Float SG 2:NiagaraFloat=1
     Float SG 3:NiagaraFloat=&RandomRangeFloat(Minimum=3, Maximum=10)
     Float SG 4:NiagaraFloat=&RandomRangeFloat(Minimum=4, Maximum=5)
     Float SG 5:NiagaraFloat=0
    )
    ScaleA=true
    ScaleRGB=true
   )
  ]
  rendererClasses [&NiagaraSpriteRendererProperties]
  renderers [
   {rendererClass=&NiagaraSpriteRendererProperties}
   {rendererClass=&NiagaraSpriteRendererProperties, rendererIndex=1}
  ]
 }
 NE_Firework_Inner {
  emitterUpdateScript [
   &EmitterState()
   &VC_FilteredSocketCount()
  ]
  particleSpawnScript [
   &self:ColorVariantID_Material_Select(VariantID:NiagaraInt32=$User.ColorVariantID) {module=ColorVariantID_Material_Select}
   &VC_PaintedVariantsOverride(
    Black=<rgba 0 0 0 2>
    Crimson=<rgba 0.4635 0 0.0126 1>
    Grey=<rgba 0.2031 0.2031 0.2031 2>
    R1 Black:LinearColor=<rgba 0.001 0.001 0.001 45>
    R1 Grey:LinearColor=<rgba 0.03 0.03 0.03 1>
    Saffron=<rgba 0.9 0.69 0 0.8>
    Use R1 Painted Overrides=true
   )
   &InitializeParticle(
    Color Mode=::NewEnumerator0
    Lifetime Max=0.9
    Lifetime Min=0.25
    Lifetime Mode=::NewEnumerator1
    Sprite Rotation Mode=::NewEnumerator2
    Sprite Size Mode=::NewEnumerator1
    Uniform Sprite Size Max=6
    UsePositionOffset=false
    Write Material Random=true
   )
   /SetVariables(Particles.SparkleRot_01:NiagaraFloat=&RandomRangeFloat(Minimum=-1, Maximum=1)) {isSetParameters=true, module=SetVariables_#2}
   &PartitionParticles(Partition Mode:ENiagara_PartitionMode_v2=::NewEnumerator5)
   &AddVelocity(Module.Use Falloff Distance=false, Override Seed=false, Velocity Mode=::NewEnumerator1, Velocity Speed=10) {module=AddVelocity002}
  ]
  particleUpdateScript [
   &ParticleState()
   /SetVariables(
    Particles.SparkleSize_01:Vector2f=&Multiply_Vector2D_ByFloat(
     Float=&Multiply_Float(
      A=&RandomRangeFloat(Minimum=35, Maximum=45)
      B=&FloatFromCurve(FloatCurve=DataInterface() {propertyValues {Curve {
          defaultValue=1
          keys [
           {interpMode=RCIM_Cubic, leaveTangentWeight=0.1, tangentMode=RCTM_User, time=0, value=0.1}
           {interpMode=RCIM_Cubic, tangentMode=RCTM_User, time=0.3, value=1}
           {arriveTangent=-3.0624, interpMode=RCIM_Cubic, leaveTangent=-3.0624, tangentMode=RCTM_User, time=1, value=0.1}
          ]
         }}})
     )
     Vector2D=<1 1>
    )
    Particles.SparkleSize_Dark:Vector2f=&Multiply_Vector2D_ByFloat(
     Float=&Multiply_Float(
      A=&RandomRangeFloat(Minimum=25, Maximum=75)
      B=&FloatFromCurve(FloatCurve=DataInterface() {propertyValues {Curve {
          defaultValue=1
          keys [
           {interpMode=RCIM_Cubic, leaveTangentWeight=0.1, tangentMode=RCTM_User, time=0, value=0.1}
           {interpMode=RCIM_Cubic, tangentMode=RCTM_User, time=0.3, value=1}
           {arriveTangent=-3.0624, interpMode=RCIM_Cubic, leaveTangent=-3.0624, tangentMode=RCTM_User, time=1, value=0}
          ]
         }}})
     )
     Vector2D=<1 1>
    )
   ) {isSetParameters=true, module=SetVariables_#3}
   &Drag(Drag=5, Use Linear Drag=true, Use Rotational Drag=true)
   &SolveForcesAndVelocity(Clamp Velocity=false, Limit Acceleration=false)
   &ScaleSpriteSizeBySpeed(Max Scale Factor=<1.4 6>, Min Scale Factor=<1 0.1>, Velocity Threshold=50)
   &ScaleColor(
    Scale Alpha=&FloatFromCurve(FloatCurve=DataInterface() {propertyValues {Curve {
        defaultValue=1
        keys [{interpMode=RCIM_Cubic, tangentMode=RCTM_User, time=0.5, value=1}, {interpMode=RCIM_Cubic, tangentMode=RCTM_User, time=1, value=0}]
       }}})
    ScaleA=true
    ScaleRGB=true
   )
  ]
  rendererClasses [&NiagaraSpriteRendererProperties]
  renderers [
   {rendererClass=&NiagaraSpriteRendererProperties}
   {rendererClass=&NiagaraSpriteRendererProperties, rendererIndex=1}
   {rendererClass=&NiagaraSpriteRendererProperties, rendererIndex=2}
   {rendererClass=&NiagaraSpriteRendererProperties, rendererIndex=3}
  ]
 }
 NE_Firework_Outter {
  emitterUpdateScript [
   &EmitterState()
   &VC_FilteredSocketCount()
  ]
  particleSpawnScript [
   &self:ColorVariantID_Material_Select(VariantID:NiagaraInt32=$User.ColorVariantID) {module=ColorVariantID_Material_Select}
   &VC_PaintedVariantsOverride(
    Black=<rgba 0 0 0 10>
    Crimson=<rgba 0.4635 0 0.0126 1>
    Grey=<rgba 0.2031 0.2031 0.2031 15>
    R1 Black:LinearColor=<rgba 0.001 0.001 0.001 45>
    R1 Grey:LinearColor=<rgba 0.025 0.025 0.025 1>
    R1 Saffron:LinearColor=<rgba 0.9 0.69 0 0.8>
    Use R1 Painted Overrides=true
   )
   &InitializeParticle(
    Color Mode=::NewEnumerator0
    Lifetime Max=0.8
    Lifetime Min=0.45
    Lifetime Mode=::NewEnumerator1
    Sprite Rotation Mode=::NewEnumerator2
    Sprite Size Mode=::NewEnumerator1
    Uniform Sprite Size Max=5
    Uniform Sprite Size Min=4
    UsePositionOffset=false
    Write Material Random=true
   )
   /SetVariables(Particles.SparkleRot_01:NiagaraFloat=&RandomRangeFloat(Minimum=-1, Maximum=1)) {isSetParameters=true, module=SetVariables_#2}
   &PartitionParticles(Partition Mode:ENiagara_PartitionMode_v2=::NewEnumerator5)
  ]
  particleUpdateScript [
   &ParticleState()
   /SetVariables(
    Particles.SparkleSize_01:Vector2f=&Multiply_Vector2D_ByFloat(
     Float=&Multiply_Float(
      A=&RandomRangeFloat(Minimum=20, Maximum=30)
      B=&FloatFromCurve(FloatCurve=DataInterface() {propertyValues {Curve {
          defaultValue=1
          keys [
           {arriveTangent=13.3293, interpMode=RCIM_Cubic, leaveTangent=13.3293, tangentMode=RCTM_User, time=0, value=0.1}
           {interpMode=RCIM_Cubic, tangentMode=RCTM_User, time=0.1, value=1}
           {arriveTangent=-3.0624, interpMode=RCIM_Cubic, leaveTangent=-3.0624, tangentMode=RCTM_User, time=1, value=0.1}
          ]
         }}})
     )
     Vector2D=<1 1>
    )
    Particles.SparkleSize_Dark:Vector2f=&Multiply_Vector2D_ByFloat(
     Float=&Multiply_Float(
      A=&RandomRangeFloat(Minimum=20, Maximum=40)
      B=&FloatFromCurve(FloatCurve=DataInterface() {propertyValues {Curve {
          defaultValue=1
          keys [
           {arriveTangent=5.4199, interpMode=RCIM_Cubic, leaveTangent=5.4199, leaveTangentWeight=0.1, tangentMode=RCTM_User, time=0, value=0.1}
           {interpMode=RCIM_Cubic, tangentMode=RCTM_User, time=0.2, value=1}
           {arriveTangent=-3.0624, interpMode=RCIM_Cubic, leaveTangent=-3.0624, tangentMode=RCTM_User, time=1, value=0}
          ]
         }}})
     )
     Vector2D=<1 1>
    )
   ) {isSetParameters=true, module=SetVariables_#3}
   &Drag(Drag=7, Use Linear Drag=true, Use Rotational Drag=true)
   &ScaleSpriteSize(
    Scale Factor=&Lerp_Vector2(
     A=<1 0.05>
     B=<1.5 3.5>
     Alpha=&FloatFromCurve(FloatCurve=DataInterface() {propertyValues {Curve {keys [<at 0.4 1 RCIM_Cubic>,<at 0.6 0>]}}})
    )
    Scale Sprite Size Mode=::NewEnumerator2
   )
   &CurlNoiseForce(
    Bias Noise=false
    Noise Strength=&Multiply_Float(
     A=1000
     B=&FloatFromCurve(FloatCurve=DataInterface() {propertyValues {Curve {keys [<at 0.6 0>,<at 0.85 1>]}}})
    )
    Pan Noise=false
   )
   &SolveForcesAndVelocity(Clamp Velocity=false, Limit Acceleration=false)
   &ScaleColor(
    Scale Alpha=&Multiply_Float(
     A=&FloatFromCurve(
      FloatCurve=DataInterface() {propertyValues {Curve {
         defaultValue=1
         keys [{interpMode=RCIM_Cubic, tangentMode=RCTM_User, time=0.2, value=1}, {interpMode=RCIM_Cubic, tangentMode=RCTM_User, time=1, value=0}]
        }}}
      Scale Curve=0.5
     )
     B=1
    )
    ScaleA=true
    ScaleRGB=true
   )
  ]
  rendererClasses [&NiagaraSpriteRendererProperties]
  renderers [
   {rendererClass=&NiagaraSpriteRendererProperties}
   {rendererClass=&NiagaraSpriteRendererProperties, rendererIndex=1}
   {rendererClass=&NiagaraSpriteRendererProperties, rendererIndex=2}
   {rendererClass=&NiagaraSpriteRendererProperties, rendererIndex=3}
  ]
 }
 NE_Firework_Start {
  simTarget=GPUComputeSim
  emitterUpdateScript [
   &EmitterState(Life Cycle Mode=::NewEnumerator1, Loop Duration=1, UseLoopDelay=false)
   &VC_FilteredSocketCount()
   &SpawnBurst_Instantaneous(Loop Count Limit=0, Spawn Count=25, Spawn Time=0.18, Use Spawn Probability=false, UseLoopCountLimit=false)
   &SpawnBurst_Instantaneous(Loop Count Limit=0, Spawn Count=75, Spawn Group=1, Use Spawn Probability=false, UseLoopCountLimit=false) {module=SpawnBurst_Instantaneous001}
  ]
  particleSpawnScript [
   &self:ColorVariantID_Material_Select(VariantID:NiagaraInt32=$User.ColorVariantID) {module=ColorVariantID_Material_Select}
   &VC_PaintedVariantsOverride(
    Black=<rgba 0 0 0 1>
    BurntSienna=&Multiply_LinearColorByFloat(Linear Color=<rgba 0.25 0.1022 0.0625 1>, Float=&RandomRangeFloat(Minimum=150, Maximum=200))
    Cobalt=&Multiply_LinearColorByFloat(Linear Color=<rgba 0 0.0185 1 1>, Float=&RandomRangeFloat(Minimum=150, Maximum=200))
    Crimson=&Multiply_LinearColorByFloat(Linear Color=<rgba 0.4635 0 0.0126 1>, Float=&RandomRangeFloat(Minimum=150, Maximum=200))
    DefaultColor=&Multiply_LinearColorByFloat(Linear Color=<rgba 0.3115 0.2108 1 1>, Float=&RandomRangeFloat(Minimum=150, Maximum=200))
    ForestGreen=&Multiply_LinearColorByFloat(Linear Color=<rgba 0 0.5 0.0253 1>, Float=&RandomRangeFloat(Minimum=150, Maximum=200))
    Grey=<rgba 0.05 0.05 0.05 1>
    Lime=&Multiply_LinearColorByFloat(Linear Color=<rgba 0.2943 0.5885 0 1>, Float=&RandomRangeFloat(Minimum=150, Maximum=200))
    Orange=&Multiply_LinearColorByFloat(Linear Color=<rgba 0.6724 0.2346 0 1>, Float=&RandomRangeFloat(Minimum=150, Maximum=200))
    Pink=&Multiply_LinearColorByFloat(Linear Color=<rgba 0.5052 0.0605 0.3435 1>, Float=&RandomRangeFloat(Minimum=150, Maximum=200))
    Purple=&Multiply_LinearColorByFloat(Linear Color=<rgba 0.2789 0 0.6795 1>, Float=&RandomRangeFloat(Minimum=150, Maximum=200))
    Saffron=/Niagara/DynamicInputs/TypeConversions/MakeLinearColorFromFloat(Value:NiagaraFloat=&RandomRangeFloat(Minimum=150, Maximum=200))
    SkyBlue=&Multiply_LinearColorByFloat(Linear Color=<rgba 0 0.357 0.5365 1>, Float=&RandomRangeFloat(Minimum=150, Maximum=200))
    TitaniumWhite=&Multiply_LinearColorByFloat(Linear Color=<rgba 0.75 0.75 0.75 1>, Float=&RandomRangeFloat(Minimum=150, Maximum=200))
   )
   &InitializeParticle(
    Color Mode=::NewEnumerator0
    Lifetime=&MakeCustomFloatFromBool(
     Bool=$User.bIsFrontendPreview
     True Float=&RandomRangeFloat(Minimum=0.2, Maximum=0.5)
     False Float=&RandomRangeFloat(Minimum=0.1, Maximum=0.4)
    )
    Sprite Size Mode=::NewEnumerator1
    Uniform Sprite Size Max=6
    Uniform Sprite Size Min=3
    UsePositionOffset=false
    Write Material Random=true
   )
   &PartitionParticles(Partition Mode:ENiagara_PartitionMode_v2=::NewEnumerator5)
   &SkeletalMeshLocation(
    Bone / Socket Index=$Particles.PartitionParticles.Partition
    Bone / Socket Sampling Mode=::NewEnumerator5
    MaskSpawnGroup=false
    Mesh Sampling Type=::NewEnumerator6
    Skeletal Mesh=$System.SkeletalMesh
   )
   &ShapeLocation(
    MaskSpawnGroup=false
    Ring / Disc Distribution Mode:ENiagaraSphereDistributionMode=::NewEnumerator3
    Ring Radius=35
    Rotation Mode=::NewEnumerator1
    Shape Primitive=::NewEnumerator4
    UseOwnerScale=true
   )
   &AddVelocity(
    Cone Angle=85
    Cone Axis=<-1 0 0>
    Inner Cone Angle=10
    Module.UseConeFalloff=false
    Module.UseVelDistribution=false
    Override Seed=false
    Velocity Mode=::NewEnumerator2
    Velocity Speed=&MakeCustomFloatFromBool(
     Bool=$User.bIsFrontendPreview
     True Float=&RandomRangeFloat(Minimum=500, Maximum=3000)
     False Float=&RandomRangeFloat(Minimum=500, Maximum=3000)
    )
   ) {module=AddVelocity001}
   &AddVelocity(
    Module.Use Falloff Distance=false
    Override Seed=false
    Velocity Mode=::NewEnumerator1
    Velocity Origin=$Output.SkeletalMeshLocation.SampledPositionWithoutOffset
    Velocity Speed=&MakeCustomFloatFromBool(
     Bool=$User.bIsFrontendPreview
     True Float=&RandomRangeFloat(Minimum=0, Maximum=0)
     False Float=&RandomRangeFloat(Minimum=500, Maximum=1000)
    )
   ) {enabled=false}
  ]
  particleUpdateScript [
   &ParticleState()
   &Drag(
    Drag=&MakeCustomFloatFromBool(
     Bool=true
     True Float=&RandomRangeFloat(Minimum=5, Maximum=10)
     False Float=&RandomRangeFloat(Minimum=5, Maximum=6)
    )
    Use Linear Drag=true
    Use Rotational Drag=true
   )
   &GravityForce(Gravity=&Multiply_VectorByFloat(Vector=<1 1 -980>, Float=0.2))
   &SolveForcesAndVelocity(Clamp Velocity=false, Limit Acceleration=false)
   &ScaleSpriteSizeBySpeed(
    Max Scale Factor=&RandomRangeVector2D(Minimum=<1 10>, Maximum=<1 15>)
    Min Scale Factor=<0.75 1>
    Velocity Threshold:NiagaraFloat=&MakeCustomFloatFromBool(Bool=$User.bIsFrontendPreview, True Float=1000, False Float=500)
   )
   &ScaleColor(
    Scale Alpha=&FloatFromCurve(FloatCurve=DataInterface() {propertyValues {Curve {
        defaultValue=0
        keys [
         {interpMode=RCIM_Cubic, tangentMode=RCTM_User, time=0, value=1}
         {arriveTangentWeight=0.5157, interpMode=RCIM_Cubic, tangentMode=RCTM_User, time=1, value=0}
        ]
       }}})
    ScaleA=true
    ScaleRGB=true
   )
  ]
  rendererClasses [&NiagaraSpriteRendererProperties]
  renderers [
   {rendererClass=&NiagaraSpriteRendererProperties}
   {rendererClass=&NiagaraSpriteRendererProperties, rendererIndex=1}
  ]
 }
 NE_VC_Boost_Sprite_LargeBurst_Parent {
  emitterUpdateScript [
   &EmitterState()
   &VC_FilteredSocketCount()
   &SpawnRate(
    SpawnRate=&MakeCustomFloatFromBool(Bool=$User.bIsFrontendPreview, True Float=32, False Float=65)
    Use Spawn Probability=false
   )
   &SpawnRate(
    SpawnGroup=1
    SpawnRate=&MakeCustomFloatFromBool(Bool=$User.bIsFrontendPreview, True Float=32, False Float=0)
    Use Spawn Probability=false
   ) {module=SpawnRate001}
  ]
  particleSpawnScript [
   &self:ColorVariantID_Material_Select(VariantID:NiagaraInt32=$User.ColorVariantID) {module=ColorVariantID_Material_Select}
   &VC_PaintedVariantsOverride(
    Black=<rgba 0 0 0 1>
    BurntSienna=<rgba 2 0.8174 0.5 1>
    Cobalt=<rgba 0 0.037 2 1>
    Crimson=<rgba 2 0 0.0542 1>
    DefaultColor=&ColorFromCurve(
     ColorCurve=DataInterface() {propertyValues {
       AlphaCurve {keys [<at 0 2>]}
       BlueCurve {keys [<at 0.1 0.01>, <at 0.45 2>, {arriveTangentWeight=0.0898, leaveTangentWeight=0.3141, time=0.55, value=2}, <at 0.9 2>]}
       GreenCurve {keys [<at 0.1 0>,<at 0.45 0.2>,<at 0.55 0.2>,<at 0.9 0.2>]}
       RedCurve {keys [<at 0.1 2>, <at 0.45 0.8>, {arriveTangent=-1.7778, leaveTangent=-1.7778, time=0.55, value=0.8}, <at 0.9 0>]}
      }}
     CurveIndex=&RandomRangeFloat(Minimum=0, Maximum=1)
    )
    ForestGreen=<rgba 0 2 0.1011 1>
    Grey=<rgba 0.05 0.05 0.05 2>
    Lime=<rgba 1 2 0 1>
    Orange=<rgba 2 0.6976 0 1>
    Pink=<rgba 2 0.2396 1.3598 1>
    Purple=<rgba 0.8208 0 2 1>
    Saffron=<rgba 2 1.5333 0 1>
    SkyBlue=<rgba 0 1.3308 2 1>
    TitaniumWhite=<rgba 1 1 1 1>
   )
   &InitializeParticle(
    Color Mode=::NewEnumerator0
    Lifetime=&MakeCustomFloatFromBool(
     Bool=$User.bIsFrontendPreview
     True Float=&Lerp_Float(
      A=0.12
      B=0.5
      Alpha=&VC_FloatBySpawnGroup(
       SpawnGroup or ExecIndex:NiagaraInt32=0
       CustomInt:NiagaraInt32=0
       MaximumSpawnGroups:NiagaraInt32=2
       Float SG 0:NiagaraFloat=&FloatFromCurve(
        FloatCurve=DataInterface() {propertyValues {Curve {keys [
            <at 0 0>
            {arriveTangent=2.5, leaveTangent=2.5, time=0.2, value=1}
            <at 0.201 0>
            <at 0.4 1>
            <at 0.401 0>
            <at 0.6 1>
            <at 0.601 0>
            <at 0.8 1>
            <at 0.801 0>
            <at 1 1>
           ]}}}
        CurveIndex=$System.NormalizedLoopAge
       )
       Float SG 1:NiagaraFloat=&FloatFromCurve(
        FloatCurve=DataInterface() {propertyValues {Curve {keys [
            {arriveTangent=2.5, time=0, value=0.5}
            <at 0.1 1>
            <at 0.101 0>
            <at 0.3 1>
            <at 0.301 0>
            <at 0.5 1>
            <at 0.501 0>
            <at 0.7 1>
            <at 0.701 0>
            <at 0.9 1>
            <at 0.901 0>
            <at 1 0.5>
           ]}}}
        CurveIndex=$System.NormalizedLoopAge
       )
       Float SG 2:NiagaraFloat=0
       Float SG 3:NiagaraFloat=0
       Float SG 4:NiagaraFloat=0
       Float SG 5:NiagaraFloat=0
      )
     )
     False Float=&RandomRangeFloat(Minimum=0.01, Maximum=0.25)
    )
    Ribbon Width=&RandomRangeFloat(Minimum=0.12, Maximum=0.5)
    Ribbon Width Mode=::NewEnumerator1
    Sprite Rotation Mode=::NewEnumerator2
    Sprite Size Max=<10 75>
    Sprite Size Min=<8 50>
    Sprite Size Mode=::NewEnumerator2
    UsePositionOffset=false
    Write Material Random=true
   )
   &PartitionParticles()
   &SkeletalMeshLocation(
    Bone / Socket Index=$Particles.PartitionParticles.Partition
    Bone / Socket Sampling Mode=::NewEnumerator5
    MaskSpawnGroup=false
    Mesh Sampling Type=::NewEnumerator6
    Skeletal Mesh=$System.SkeletalMesh
   )
   &ShapeLocation(
    MaskSpawnGroup=false
    Override Seed=false
    Sphere Radius=&MakeCustomFloatFromBool(Bool=$User.bIsFrontendPreview, True Float=1, False Float=100)
    UseOwnerScale=true
   )
   &AddVelocity(
    Cone Angle=&MakeCustomFloatFromBool(Bool=$User.bIsFrontendPreview, True Float=18, False Float=45)
    Cone Axis=<-1 0 0.05>
    Module.UseConeFalloff=false
    Module.UseVelDistribution=false
    Override Seed=false
    Velocity Mode=::NewEnumerator2
    Velocity Speed=&MakeCustomFloatFromBool(
     Bool=$User.bIsFrontendPreview
     True Float=&RandomRangeFloat(Minimum=1200, Maximum=2500)
     False Float=&RandomRangeFloat(Minimum=100, Maximum=300)
    )
   )
   /SetVariables(
    Particles.FW_ringRot:Vector3f=&RandomRangeVector(Minimum=<-1 -1 -1>, Maximum=<1 1 1>)
    Particles.FW_velocityMult:NiagaraFloat=&RandomRangeFloat(Minimum=0.1, Maximum=1)
   ) {isSetParameters=true}
  ]
  particleUpdateScript [
   &ParticleState()
   &SolveForcesAndVelocity(Clamp Velocity=false, Limit Acceleration=false)
   &GenerateDeathEvent(
    "Float to Send as Float 2 (RandomNormalizedFloat)":NiagaraFloat=$Particles.FW_velocityMult
    Use Event Probability=false
    "Vector to Send as Vector 3 (Acceleration)":Vector3f=$Particles.FW_ringRot
   )
  ]
  rendererClasses [&NiagaraSpriteRendererProperties]
  renderers [
   {rendererClass=&NiagaraSpriteRendererProperties}
   {rendererClass=&NiagaraSpriteRendererProperties, rendererIndex=1}
  ]
 }
 SmallDazzle {
  emitterUpdateScript [
   &EmitterState(Life Cycle Mode=::NewEnumerator1, Loop Duration=1, UseLoopDelay=false)
   &SpawnRate(SpawnRate=20, Use Spawn Probability=false) {enabled=false}
   &SpawnParticlesFromOtherEmitter(
    Attribute Reader=DataInterface() {propertyValues {EmitterBinding {bindingMode=Other, emitterName=NE_VC_Boost_Sprite_LargeBurst_Parent}}}
    Use Maximum Age=false
    Use Minimum Age=false
   )
  ]
  particleSpawnScript [
   &InitializeParticle(
    Color=&ColorFromCurve(ColorCurve=DataInterface() {propertyValues {
       AlphaCurve {keys [<at 0 1 RCIM_Cubic>,<at 1 0 RCIM_Cubic>]}
       BlueCurve {keys [<at 0 0.329 RCIM_Cubic>,<at 0.4579 1>,<at 1 0 RCIM_Cubic>]}
       GreenCurve {keys [<at 0 0 RCIM_Cubic>,<at 0.4579 0.4776>,<at 1 0.7965 RCIM_Cubic>]}
       RedCurve {keys [<at 0 1 RCIM_Cubic>,<at 0.4579 0>,<at 1 1 RCIM_Cubic>]}
      }})
    Lifetime Max=0.9
    Lifetime Min=0.25
    Lifetime Mode=::NewEnumerator1
    Sprite Rotation Mode=::NewEnumerator2
    Sprite Size Mode=::NewEnumerator1
    Uniform Sprite Size Max=6
    UsePositionOffset=false
    Write Material Random=true
   )
   /SetVariables(Particles.SparkleRot_01:NiagaraFloat=&RandomRangeFloat(Minimum=-1, Maximum=1)) {isSetParameters=true}
   &PartitionParticles(Partition Mode:ENiagara_PartitionMode_v2=::NewEnumerator5)
   &AddVelocity(Module.Use Falloff Distance=false, Override Seed=false, Velocity Mode=::NewEnumerator1, Velocity Speed=1) {module=AddVelocity002}
   &ShapeLocation(MaskSpawnGroup=false, Override Seed=false, Sphere Radius=10, UseOwnerScale=true)
   &SampleParticlesFromOtherEmitter(Color Sampling=::NewEnumerator0)
  ]
  particleUpdateScript [
   &ParticleState()
   /SetVariables(
    Particles.SparkleSize_01:Vector2f=&Multiply_Vector2D_ByFloat(
     Float=&Multiply_Float(
      A=&RandomRangeFloat(Minimum=35, Maximum=45)
      B=&FloatFromCurve(FloatCurve=DataInterface() {propertyValues {Curve {
          defaultValue=1
          keys [
           {interpMode=RCIM_Cubic, leaveTangentWeight=0.1, tangentMode=RCTM_User, time=0, value=0.1}
           {interpMode=RCIM_Cubic, tangentMode=RCTM_User, time=0.3, value=1}
           {arriveTangent=-3.0624, interpMode=RCIM_Cubic, leaveTangent=-3.0624, tangentMode=RCTM_User, time=1, value=0.1}
          ]
         }}})
     )
     Vector2D=<1 1>
    )
    Particles.SparkleSize_Dark:Vector2f=&Multiply_Vector2D_ByFloat(
     Float=&Multiply_Float(
      A=&RandomRangeFloat(Minimum=25, Maximum=75)
      B=&FloatFromCurve(FloatCurve=DataInterface() {propertyValues {Curve {
          defaultValue=1
          keys [
           {interpMode=RCIM_Cubic, leaveTangentWeight=0.1, tangentMode=RCTM_User, time=0, value=0.1}
           {interpMode=RCIM_Cubic, tangentMode=RCTM_User, time=0.3, value=1}
           {arriveTangent=-3.0624, interpMode=RCIM_Cubic, leaveTangent=-3.0624, tangentMode=RCTM_User, time=1, value=0}
          ]
         }}})
     )
     Vector2D=<1 1>
    )
   ) {isSetParameters=true}
   &Drag(Drag=5, Use Linear Drag=true, Use Rotational Drag=true)
   &ScaleVelocity(Velocity Scale=&RandomRangeVector(Minimum=<-100 -100 -100>, Maximum=<100 100 100>))
   &SolveForcesAndVelocity(Clamp Velocity=false, Limit Acceleration=false)
   &ScaleSpriteSizeBySpeed(Max Scale Factor=<1.4 6>, Min Scale Factor=<1 0.1>, Velocity Threshold=50)
   &ScaleColor(
    Scale Alpha=&FloatFromCurve(FloatCurve=DataInterface() {propertyValues {Curve {
        defaultValue=1
        keys [{interpMode=RCIM_Cubic, tangentMode=RCTM_User, time=0.5, value=1}, {interpMode=RCIM_Cubic, tangentMode=RCTM_User, time=1, value=0}]
       }}})
    ScaleA=true
    ScaleRGB=true
   )
   &ShapeLocation(MaskSpawnGroup=true, Shape Origin=$Particles.Position, Sphere Distribution=::NewEnumerator3, Sphere Radius=1, UseOwnerScale=true) {module=ShapeLocation001}
  ]
  rendererClasses [&NiagaraSpriteRendererProperties]
  renderers [
   {rendererClass=&NiagaraSpriteRendererProperties}
   {rendererClass=&NiagaraSpriteRendererProperties, rendererIndex=1}
   {rendererClass=&NiagaraSpriteRendererProperties, rendererIndex=2}
   {rendererClass=&NiagaraSpriteRendererProperties, rendererIndex=3}
  ]
 }
 SmallDazzleOrig {
  emitterUpdateScript [
   &EmitterState()
   &VC_FilteredSocketCount()
  ]
  particleSpawnScript [
   &self:ColorVariantID_Material_Select(VariantID:NiagaraInt32=$User.ColorVariantID) {module=ColorVariantID_Material_Select}
   &VC_PaintedVariantsOverride(
    Black=<rgba 0 0 0 2>
    Crimson=<rgba 0.4635 0 0.0126 1>
    Grey=<rgba 0.2031 0.2031 0.2031 2>
    R1 Black:LinearColor=<rgba 0.001 0.001 0.001 45>
    R1 Grey:LinearColor=<rgba 0.03 0.03 0.03 1>
    Saffron=<rgba 0.9 0.69 0 0.8>
    Use R1 Painted Overrides=true
   )
   &InitializeParticle(
    Color Mode=::NewEnumerator0
    Lifetime Max=0.9
    Lifetime Min=0.25
    Lifetime Mode=::NewEnumerator1
    Sprite Rotation Mode=::NewEnumerator2
    Sprite Size Mode=::NewEnumerator1
    Uniform Sprite Size Max=6
    UsePositionOffset=false
    Write Material Random=true
   )
   /SetVariables(Particles.SparkleRot_01:NiagaraFloat=&RandomRangeFloat(Minimum=-1, Maximum=1)) {isSetParameters=true, module=SetVariables_#2}
   &PartitionParticles(Partition Mode:ENiagara_PartitionMode_v2=::NewEnumerator5)
   &AddVelocity(Module.Use Falloff Distance=false, Override Seed=false, Velocity Mode=::NewEnumerator1, Velocity Speed=10) {module=AddVelocity002}
  ]
  particleUpdateScript [
   &ParticleState()
   /SetVariables(
    Particles.SparkleSize_01:Vector2f=&Multiply_Vector2D_ByFloat(
     Float=&Multiply_Float(
      A=&RandomRangeFloat(Minimum=35, Maximum=45)
      B=&FloatFromCurve(FloatCurve=DataInterface() {propertyValues {Curve {
          defaultValue=1
          keys [
           {interpMode=RCIM_Cubic, leaveTangentWeight=0.1, tangentMode=RCTM_User, time=0, value=0.1}
           {interpMode=RCIM_Cubic, tangentMode=RCTM_User, time=0.3, value=1}
           {arriveTangent=-3.0624, interpMode=RCIM_Cubic, leaveTangent=-3.0624, tangentMode=RCTM_User, time=1, value=0.1}
          ]
         }}})
     )
     Vector2D=<1 1>
    )
    Particles.SparkleSize_Dark:Vector2f=&Multiply_Vector2D_ByFloat(
     Float=&Multiply_Float(
      A=&RandomRangeFloat(Minimum=25, Maximum=75)
      B=&FloatFromCurve(FloatCurve=DataInterface() {propertyValues {Curve {
          defaultValue=1
          keys [
           {interpMode=RCIM_Cubic, leaveTangentWeight=0.1, tangentMode=RCTM_User, time=0, value=0.1}
           {interpMode=RCIM_Cubic, tangentMode=RCTM_User, time=0.3, value=1}
           {arriveTangent=-3.0624, interpMode=RCIM_Cubic, leaveTangent=-3.0624, tangentMode=RCTM_User, time=1, value=0}
          ]
         }}})
     )
     Vector2D=<1 1>
    )
   ) {isSetParameters=true, module=SetVariables_#3}
   &Drag(Drag=5, Use Linear Drag=true, Use Rotational Drag=true)
   &SolveForcesAndVelocity(Clamp Velocity=false, Limit Acceleration=false)
   &ScaleSpriteSizeBySpeed(Max Scale Factor=<1.4 6>, Min Scale Factor=<1 0.1>, Velocity Threshold=50)
   &ScaleColor(
    Scale Alpha=&FloatFromCurve(FloatCurve=DataInterface() {propertyValues {Curve {
        defaultValue=1
        keys [{interpMode=RCIM_Cubic, tangentMode=RCTM_User, time=0.5, value=1}, {interpMode=RCIM_Cubic, tangentMode=RCTM_User, time=1, value=0}]
       }}})
    ScaleA=true
    ScaleRGB=true
   )
  ]
  rendererClasses [&NiagaraSpriteRendererProperties]
  renderers [
   {rendererClass=&NiagaraSpriteRendererProperties}
   {rendererClass=&NiagaraSpriteRendererProperties, rendererIndex=1}
   {rendererClass=&NiagaraSpriteRendererProperties, rendererIndex=2}
   {rendererClass=&NiagaraSpriteRendererProperties, rendererIndex=3}
  ]
 }
}
summary {
 systemName=NS_Boost_Fireworks
 emitters [
  {emitterName=NE_VC_Boost_Sprite_LargeBurst_Parent, rendererClasses [&NiagaraSpriteRendererProperties]}
  {emitterName=NE_Firework_Inner, rendererClasses [&NiagaraSpriteRendererProperties]}
  {emitterName=NE_Firework_Fill, simTarget=GPUComputeSim, rendererClasses [&NiagaraSpriteRendererProperties]}
  {emitterName=NE_Firework_Outter, rendererClasses [&NiagaraSpriteRendererProperties]}
  {emitterName=NE_Firework_Start, simTarget=GPUComputeSim, rendererClasses [&NiagaraSpriteRendererProperties]}
  {emitterName=NE_Firework_Fill001, rendererClasses [&NiagaraSpriteRendererProperties]}
  {emitterName=SmallDazzleOrig, rendererClasses [&NiagaraSpriteRendererProperties]}
  {emitterName=SmallDazzle, rendererClasses [&NiagaraSpriteRendererProperties]}
 ]
 userVariables [
  User.bIsFrontendPreview {type=NiagaraBool}
  User.bIsPainted {type=NiagaraBool}
  User.ColorVariantID {type=NiagaraInt32}
  User.PaintedColor {type=LinearColor, description="", defaultValue {struct=LinearColor, value=<rgba 1 1 1 1>}}
  User.PaintedScalar {type=NiagaraFloat, defaultValue=1, description=""}
  User.PortCount {type=NiagaraInt32, defaultValue=1}
 ]
}
